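; Gameplay ability for a Paprika med consumable that restores health and/or shields incrementally during its use montage.
; Triggering manually consumes one stack, heals immediately, then repeats on a timer spreading max ticks across the remaining montage.
; Each tick applies per-level amounts via gameplay effect; authority ends the ability once maxed.
; Ending removes empty stacks, sends consume events, commits MedKit cooldown and can toss the item as a projectile.
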(event Ability|EventActivateAbility
  ;[0.03429 0.3069 0.7899] Ensure healing level begins at zero
  (|CallParentFunction)
  (if (> (ScalableFloat|GetValueasDouble (Variables|Default|GetCancelSlide)) 0)
    (Ability|GameplayEffect|AddLooseGameplayTags (CallFunction|GetAvatarasFortPlayerPawn) (Variables|Default|GetcancelSlidingTag)))
  (Variables|Default|SetHealingLevel 0))

(event EventTriggerAbilityInternal
  (bind _returnvalue (CallFunction|GetAvatarasFortPlayerPawn))
  (bind _returnvalue_2 (Fortnite|GetFortPlayerControllerfromActor _returnvalue))
  ;[0.03429 0.3069 0.7899] Trigger ability override to setup healing timer / THIS CONSUMES THE STACK
  (Variables|Default|SetAbilityHasBeenTriggered true)
  (bind _bsuccess (CallFunction|CommitCostManually))
  (if (and _bsuccess (== (Variables|Default|GetHealingLevel) 0))
    ;;[0 0.8244 1] Do one tick immediately
    (CallFunction|DoIncrementalHealing)
    (CallFunction|SendInitialHealEvent)
    (if (not (Variables|Default|GetIsMaxedOut?))
      (bind _returnvalue_3 (Utilities|Time|SetTimerbyEvent (delegate Custom|DoIncrementalHealing) (CallFunction|GetTickRate) true))
      (Variables|Default|SetHealingTimer _returnvalue_3)
      (CallFunction|OnMontageTriggeredPostCommitSuccess))
    ;[0 0.5221 1] Quests
    (GameplayTags|AppendGameplayTagContainers (Variables|Default|GetQuestWeaponTags) (Variables|Default|GetQuestTags))
    (Utilities|IsValid _returnvalue_2
      (:Is%20Valid
        (Quest|SendComplexCustomStatEvent (Quests|GetQuestManager _returnvalue_2 "Invalid") 0 (Variables|Default|GetQuestSourceTags) (Variables|Default|GetQuestWeaponTags) 1)))
    (else
      (Ability|EndAbility))))

(event EventTriggerAbilityInternalIfNotYetTriggered
  (bind _returnvalue (CallFunction|GetAvatarasFortPlayerPawn))
  (bind _returnvalue_2 (Fortnite|GetFortPlayerControllerfromActor _returnvalue))
  ;[0.03429 0.3069 0.7899] Trigger ability override to setup healing timer / THIS CONSUMES THE STACK
  (if (not (Variables|Default|GetAbilityHasBeenTriggered))
    (Variables|Default|SetAbilityHasBeenTriggered true)
    (bind _bsuccess (CallFunction|CommitCostManually))
    (if (and _bsuccess (== (Variables|Default|GetHealingLevel) 0))
      ;;[0 0.8244 1] Do one tick immediately
      (CallFunction|DoIncrementalHealing)
      (CallFunction|SendInitialHealEvent)
      (if (not (Variables|Default|GetIsMaxedOut?))
        (bind _returnvalue_3 (Utilities|Time|SetTimerbyEvent (delegate Custom|DoIncrementalHealing) (CallFunction|GetTickRate) true))
        (Variables|Default|SetHealingTimer _returnvalue_3)
        (CallFunction|OnMontageTriggeredPostCommitSuccess))
      ;[0 0.5221 1] Quests
      (GameplayTags|AppendGameplayTagContainers (Variables|Default|GetQuestWeaponTags) (Variables|Default|GetQuestTags))
      (Utilities|IsValid _returnvalue_2
        (:Is%20Valid
          (Quest|SendComplexCustomStatEvent (Quests|GetQuestManager _returnvalue_2 "Invalid") 0 (Variables|Default|GetQuestSourceTags) (Variables|Default|GetQuestWeaponTags) 1)))
      (else
        (Ability|EndAbility)))))

(event Custom|DoIncrementalHealing
  ;[0.03429 0.3069 0.7899] Do healing per tick
  (CallFunction|HealthCalc)
  (CallFunction|ShieldCalc)
  (bind _effect_out (CallFunction|SetupGE))
  (Ability|ApplyGameplayEffectSpecToOwner _effect_out)
  (Ability|ExecuteGameplayCueOnOwner (Variables|Default|GetGC_IncrementApplied))
  (Math|Integer|IncrementInt (Variables|Default|GetHealingLevel)
    (:%20%20
      ;[0.03429 0.3069 0.7899] If we have no health/shield to restore, end anim and ability
      (CallFunction|CheckHealingMaxed)
      (if (Variables|Default|GetIsMaxedOut?)
        (Ability|HasAuthority
          (:True
            (Ability|Animation|MontageStop -1.0)
            (Ability|EndAbility))) ; Only ending in authority ensures the cost is applied before ending in 1 tick edge cases
      ))))

(event Ability|EventOnEndAbility
  (params (bWasCancelled :type Boolean))
  (bind (_x _y) (Math|Rotator|BreakRotIntoAxes (Transformation|GetActorRotation (CallFunction|GetAvatarasFortPlayerPawn))))
  ;[0.03429 0.3069 0.7899] Override to cancel timer, apply cost if we did trigger the ability
  (CallFunction|CheckEmptyStack)
  (Utilities|Time|ClearandInvalidateTimerbyHandle (Variables|Default|GetHealingTimer))
  (if (Variables|Default|GetAbilityHasBeenTriggered)
    (Utilities|FlowControl|Sequence
      (:then_0
        (Default|PostConsumeEnd)
        (Ability|SendGameplayEventtoActor (Variables|Default|GetPlayerPawn) "(TagName=\"Event.ConsumedHeal\")"))
      (:then_1
        (bind _ismaxedout (CallFunction|CheckHealingMaxed))
        (if _ismaxedout
          (Ability|ExecuteGameplayCueOnOwner (Variables|Default|GetGC_HealFinished)))))
    (else
      _))
  (|CallParentFunction bWasCancelled)
  (CallFunction|EndWeaponAnims)
  ;[0.03429 0.3069 0.7899] Toss Weapon as Projectile
  (if (Default|CheckCanTossPrj)
    (Projectile|SpawnProjectile :ProjectileClass (Variables|Default|GetTossProjectileClass) :RequestedBy (CallFunction|GetAvatarasFortPlayerPawn) :SpawnLocation (Transformation|GetActorLocation (Ability|GetCurrentSourceWeapon)) :SpawnRotation "0, 0, 0" :SpawnDirection (Math|Conversions|RotationFromXVector (Math|Vector|RotateVectorAroundAxis (Transformation|GetActorForwardVector (CallFunction|GetAvatarasFortPlayerPawn)) (Variables|Default|GetUpProjectileRotation) _y)) :bUseDefaultPhysics false :InitialSpeed 4e+02 :GravityScale 1.0 :ChargePercent 0.0)))

(event Ability|EventCommitExecute)

(event Custom|CheckEmptyStack
  ;[1 1 1] Manual Cost Application
  (Ability|HasAuthority
    (:True
      (bind _asfort_inventory_owner_interface (Utilities|Casting|CastToFortInventoryOwnerInterface (Pawn|GetController (CallFunction|GetAvatarasFortPlayerPawn))))
      (bind _returnvalue (Inventory|GetInventoryItemwithGuid _asfort_inventory_owner_interface (Variables|Default|GetUsedInventoryGUID)))
      (Utilities|IsValid _returnvalue
        (:Is%20Valid
          (if (<= (Item|GetNuminStack _returnvalue) 0)
            (InventoryManagement|RemoveItem _asfort_inventory_owner_interface (Variables|Default|GetUsedInventoryGUID) 0 true)))))))

(fn CheckCanTossPrj
  (bind _montageinfo_montagetoplay (Utilities|Struct|BreakFortGameplayAbilityMontageInfo (Variables|GetMontageInfo_MontageToPlay)))
  (return (Math|Boolean|ANDBoolean (Variables|Default|GetShouldTossPrj?) (Variables|Default|GetAbilityHasBeenTriggered) (not (>= (Animation|Montage|MontageGetPosition (Components|SkeletalMesh|GetAnimInstance (Variables|/Script/Engine.Character|GetMesh (CallFunction|GetAvatarasFortPlayerPawn))) _montageinfo_montagetoplay) 6.75)))))

(fn CheckHealingMaxed
  ;[0.4895 0.1317 1] Check if we are either max health or max shield based on healing values
  (bind _output_get (Variables|Default|SetIsMaxedOut? (select (and (Variables|MedParentHealth|GetHealsHealth) (Variables|MedParentShields|GetHealsShields)) (and (and (Variables|MedParentHealth|GetHealsHealth) (>= (FortPawn|GetHealth (Variables|Default|GetPlayerPawn)) (FortPawn|GetMaxHealth (Variables|Default|GetPlayerPawn)))) (and (>= (FortPawn|GetShield (Variables|Default|GetPlayerPawn)) (FortPawn|GetMaxShield (Variables|Default|GetPlayerPawn))) (Variables|MedParentShields|GetHealsShields))) (or (and (Variables|MedParentHealth|GetHealsHealth) (>= (FortPawn|GetHealth (Variables|Default|GetPlayerPawn)) (FortPawn|GetMaxHealth (Variables|Default|GetPlayerPawn)))) (and (>= (FortPawn|GetShield (Variables|Default|GetPlayerPawn)) (FortPawn|GetMaxShield (Variables|Default|GetPlayerPawn))) (Variables|MedParentShields|GetHealsShields))))))
  (Debugging|PrintBool _output_get "is healing maxed = "
    (:Then
      (return _output_get))))

(fn CommitCostManually
  (bind _returnvalue (CallFunction|GetAvatarasFortPlayerPawn))
  (bind _returnvalue_2 (Weapon|GetCurrentWeapon _returnvalue))
  (bind _returnvalue_3 (Pawn|GetController _returnvalue))
  (Utilities|FlowControl|Sequence
    (:then_0
      (bind _asfort_player_controller_athena (Utilities|Casting|CastToFortPlayerControllerAthena _returnvalue_3)
        (:then
          (if (Class|FortPlayerControllerAthena|HasInfiniteConsumables _asfort_player_controller_athena)
            (return true)
            (else
              _)))
        (:CastFailed
          _))
      (Utilities|IsValid _returnvalue_2
        (:Is%20Valid
          (bind _asfort_inventory_owner_interface (Utilities|Casting|CastToFortInventoryOwnerInterface _returnvalue_3))
          (bind _returnvalue_4 (Inventory|GetInventoryItemwithGuid _asfort_inventory_owner_interface (FortWeapon|GetInventoryGUID _returnvalue_2)))
          (Utilities|IsValid _returnvalue_4
            (:Is%20Valid
              (bind _output_get (Variables|Default|SetUsedInventoryGUID (Item|GetItemGuid _returnvalue_4)))
              (InventoryManagement|RemoveItem _asfort_inventory_owner_interface _output_get 1 false true)
              (return true))))))
    (:then_1
      (return false))))

(fn EndWeaponAnims
  ;[0.4895 0.1317 1] Ensure weapon animations stop playing when ability ends
  (Animation|Montage|MontageStop (Components|SkeletalMesh|GetAnimInstance (Visual|GetWeaponMesh (Ability|GetCurrentSourceWeapon))) 0.2 (Animation|Montage|GetCurrentActiveMontage (Components|SkeletalMesh|GetAnimInstance (Visual|GetWeaponMesh (Ability|GetCurrentSourceWeapon))))))

(fn GetTickRate
  (bind _returnvalue (Animation|GetCurrentMontage))
  (bind _returnvalue_2 (Animation|GetPlayLength _returnvalue))
  (bind _hf_maxticks (Variables|Default|GetHF_MaxTicks))
  (bind _returnvalue_3 (ScalableFloat|GetValueatLevel _hf_maxticks))
  (bind _triggerduration (Variables|MedParentMontage|GetTriggerDuration))
  (bind _returnvalue_4 (- _returnvalue_2 _triggerduration))
  (bind _returnvalue_5 (Math|Float|SafeDivide _returnvalue_4 _returnvalue_3))
  (Development|PrintString (Utilities|String|Append (Utilities|String|ToString(Float) _returnvalue_4) " / " (Utilities|String|ToString(Float) _returnvalue_3) " = " (Utilities|String|ToString(Float) _returnvalue_5)))
  (return _returnvalue_5))

(fn HealthCalc
  (Variables|MedParentHealth|SetHealthHealAmount (Math|Float|SelectFloat (ScalableFloat|GetValueatLevel (Variables|MedParentHealth|GetRow_HealthAmount) (Math|Conversions|ToFloat(Integer) (Variables|Default|GetHealingLevel))) 0.0 (Variables|MedParentHealth|GetHealsHealth)))
  (return))

(fn PostConsumeEnd
  (if (GameplayTags|HasTag (Tags|GetWeaponDataGameplayTags (Ability|GetCurrentSourceWeapon)) "(TagName=\"Item.Consumable.Paprika.MedKit\")" true)
    (if (Utilities|GameplayTagContainerNotEmpty (Ability|Cooldowns|GetCooldownTags))
      (Ability|CommitAbilityCooldown false)
      (else
        _))
    (return)))

(fn SendInitialHealEvent
  (bind _returnvalue (CallFunction|GetAvatarasFortPlayerPawn))
  (bind _returnvalue_2 (Fortnite|GetFortPlayerControllerfromActor _returnvalue))
  (bind _returnvalue_3 (Weapon|GetCurrentWeapon _returnvalue))
  (Utilities|IsValid _returnvalue_3
    (:Is%20Valid
      (Utilities|IsValid _returnvalue_2
        (:Is%20Valid
          (bind _returnvalue_4 (Inventory|GetInventoryItemwithGuid _returnvalue_2 (FortWeapon|GetInventoryGUID _returnvalue_3)))
          (Utilities|IsValid _returnvalue_4
            (:Is%20Valid
              (bind _returnvalue_5 (GameplayTags|GetOwnedGameplayTags _returnvalue_4 :!class /Script/GameplayTags.GameplayTagAssetInterface))
              (Ability|SendGameplayEventtoActor _returnvalue "(TagName=\"Event.ConsumedInitialHeal\")" (Utilities|Struct|MakeGameplayEventData "(TagName=\"Event.ConsumedInitialHeal\")" 0 _returnvalue _returnvalue_4 0 "()" _returnvalue_5 (GameplayTags|GetOwnedGameplayTags _returnvalue :!class /Script/GameplayTags.GameplayTagAssetInterface))))))))))

(fn ShieldCalc
  (Variables|MedParentShields|SetShieldHealAmount (Math|Float|SelectFloat (ScalableFloat|GetValueatLevel (Variables|MedParentShields|GetRow_ShieldAmount) (Math|Conversions|ToFloat(Integer) (Variables|Default|GetHealingLevel))) 0.0 (Variables|MedParentShields|GetHealsShields)))
  (return))
